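# MetaSound source that plays the hover sound for top-bar tab buttons in the front-end UI.
# A single mono Wave Player starts on play, reads a wave asset input defaulting to the top-bar hover wav, and signals the one-shot finish when done.
# It uses the Fort_UI sound class, with a concurrency override that allows one instance and stops the oldest, so rapid hovers do not stack.
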MetaSoundSource /Game/Sounds/Fort_UI_Sounds/FrontEnd/Tabs/MSS_UI_Button_Tab_TopBar_Hover
inputs [
 Default Hover Sound {type=WaveAsset}
 UE.Source.OnPlay {type=Trigger}
]
interfaces [UE.OutputFormat.Mono:1.0, UE.Source.OneShot:1.0, UE.Source:1.1]
outputs [
 "UE.OutputFormat.Mono.Audio:0" {type=Audio}
 UE.Source.OneShot.OnFinished {type=Trigger}
]
pages [Default {
  input_defaults {Default Hover Sound=/Game/Sounds/Fort_UI_Sounds/FrontEnd/Tabs/wav/UI_MP_Hover_TopBar_01, UE.Source.OnPlay=null}
  nodes [WavePlayer_0 = UE.Wave Player:Mono(Wave Asset=^Default Hover Sound, Play=^UE.Source.OnPlay)]
  links [{dst="@UE.Source.OneShot.OnFinished", src=WavePlayer_0.On Finished}, {dst="@UE.OutputFormat.Mono.Audio:0", src=WavePlayer_0.Out Mono}]
 }]
properties {
 assetImportData=&self:AssetImportData
 bOverrideConcurrency=true
 sampleRate=48000
 soundClassObject=&Fort_UI
 virtualizationMode=Disabled
 concurrencyOverrides {maxCount=1, resolutionRule=StopOldest}
}
